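# Sound cue for catching the toy football, layering a catch sound with a score pop.
# A mixer plays one of two random catch waves at full volume alongside a quieter branch that, after a short fixed delay, plays one of three little score pops borrowed from the Score LTM sounds.
# The cue is routed to the Fort_Music_Emote sound class.
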SoundCue /FNGameplayCues/Athena/Sounds/Toys/Football/Toy_Football_Catch_Cue
root=Mixer_0
nodes [
 {id=Delay_0, children [Random_2], properties {delayMax=0.35, delayMin=0.35}}
 {id=Mixer_0, children [Delay_0,Random_0], properties {inputVolume [0.5,1]}}
 {id=Random_0, children [WavePlayer_0,WavePlayer_1], properties {weights [1,1]}}
 {id=Random_2, children [WavePlayer_5,WavePlayer_6,WavePlayer_7], properties {weights [1,1,1]}}
 {id=WavePlayer_0, properties {soundWaveAssetPtr=&/Toy_Football_Catch_01}}
 {id=WavePlayer_1, properties {soundWaveAssetPtr=&/Toy_Football_Catch_02}}
 {id=WavePlayer_5, properties {soundWaveAssetPtr=/Game/Athena/Sounds/LTMs/Score/ScoreLTM_InGamePop_Little_01}}
 {id=WavePlayer_6, properties {soundWaveAssetPtr=/Game/Athena/Sounds/LTMs/Score/ScoreLTM_InGamePop_Little_02}}
 {id=WavePlayer_7, properties {soundWaveAssetPtr=/Game/Athena/Sounds/LTMs/Score/ScoreLTM_InGamePop_Little_03}}
]
properties {maxDistance=2097152, soundClassObject=&Fort_Music_Emote}
